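# Physics asset for the male Tower Sentinel two-handed pickaxe, giving its root, handle and blade bones small capsule bodies and joints.
# Three blade bases hang off the handle with near-rigid limits, and the handle joins b_AR_Root.
# Each base branches into five or six two-segment chains.
# Tight swing and twist limits along these chains allow only slight flex, with some "a" and "b" joints tighter than the rest.
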PhysicsAsset /BRCosmetics/Weapons/FORT_Melee/Pickaxe_Tower_Sentinel_Male/Meshes/Tower_Sentinel_Male_Physics
bodies {
 Melee_TwoHanded_Handle [{shapeType=Capsule, length=2, radius=2}]
 b_AR_Root [{shapeType=Capsule, length=2, radius=2}]
 blade1_a_01 [{shapeType=Capsule, length=1, radius=1, center=<1 0 0>, rotation=<rot -90 0 0>}]
 blade1_a_02 [{shapeType=Capsule, length=1, radius=1, center=<1 0 0>, rotation=<rot -90 0 0>}]
 blade1_b_01 [{shapeType=Capsule, length=1, radius=1, center=<1 0 0>, rotation=<rot -90 0 0>}]
 blade1_b_02 [{shapeType=Capsule, length=1, radius=1, center=<1 0 0>, rotation=<rot -90 0 0>}]
 blade1_base [{shapeType=Capsule, length=1, radius=1, rotation=<rot -90 0 0>}]
 blade1_c_01 [{shapeType=Capsule, length=1, radius=1, center=<1 0 0>, rotation=<rot -90 0 0>}]
 blade1_c_02 [{shapeType=Capsule, length=1, radius=1, center=<1 0 0>, rotation=<rot -90 0 0>}]
 blade1_d_01 [{shapeType=Capsule, length=1, radius=1, center=<1 0 0>, rotation=<rot -90 0 0>}]
 blade1_d_02 [{shapeType=Capsule, length=1, radius=1, center=<1 0 0>, rotation=<rot -90 0 0>}]
 blade1_e_01 [{shapeType=Capsule, length=1, radius=1, center=<1 0 0>, rotation=<rot -90 0 0>}]
 blade1_e_02 [{shapeType=Capsule, length=1, radius=1, center=<1 0 0>, rotation=<rot -90 0 0>}]
 blade1_f_01 [{shapeType=Capsule, length=1, radius=1, center=<1 0 0>, rotation=<rot -90 0 0>}]
 blade1_f_02 [{shapeType=Capsule, length=1, radius=1, center=<1 0 0>, rotation=<rot -90 0 0>}]
 blade2_a_02 [{shapeType=Capsule, length=1, radius=1, center=<1 0 0>, rotation=<rot -90 0 0>}]
 blade2_a_03 [{shapeType=Capsule, length=1, radius=1, center=<1 0 0>, rotation=<rot -90 0 0>}]
 blade2_b_01 [{shapeType=Capsule, length=1, radius=1, center=<1 0 0>, rotation=<rot -90 0 0>}]
 blade2_b_02 [{shapeType=Capsule, length=1, radius=1, center=<1 0 0>, rotation=<rot -90 0 0>}]
 blade2_base [{shapeType=Capsule, length=1, radius=1, rotation=<rot -90 0 0>}]
 blade2_c_01 [{shapeType=Capsule, length=1, radius=1, center=<1 0 0>, rotation=<rot -90 0 0>}]
 blade2_c_02 [{shapeType=Capsule, length=1, radius=1, center=<1 0 0>, rotation=<rot -90 0 0>}]
 blade2_d_01 [{shapeType=Capsule, length=1, radius=1, center=<1 0 0>, rotation=<rot -90 0 0>}]
 blade2_d_02 [{shapeType=Capsule, length=1, radius=1, center=<1 0 0>, rotation=<rot -90 0 0>}]
 blade2_e_01 [{shapeType=Capsule, length=1, radius=1, center=<1 0 0>, rotation=<rot -90 0 0>}]
 blade2_e_02 [{shapeType=Capsule, length=1, radius=1, center=<1 0 0>, rotation=<rot -90 0 0>}]
 blade2_f_01 [{shapeType=Capsule, length=1, radius=1, center=<1 0 0>, rotation=<rot -90 0 0>}]
 blade2_f_02 [{shapeType=Capsule, length=1, radius=1, center=<1 0 0>, rotation=<rot -90 0 0>}]
 blade3_a_01 [{shapeType=Capsule, length=1, radius=1, center=<1 0 0>, rotation=<rot -90 0 0>}]
 blade3_a_02 [{shapeType=Capsule, length=1, radius=1, center=<1 0 0>, rotation=<rot -90 0 0>}]
 blade3_b_01 [{shapeType=Capsule, length=1, radius=1, center=<1 0 0>, rotation=<rot -90 0 0>}]
 blade3_b_02 [{shapeType=Capsule, length=1, radius=1, center=<1 0 0>, rotation=<rot -90 0 0>}]
 blade3_base [{shapeType=Capsule, length=2, radius=2}]
 blade3_c_01 [{shapeType=Capsule, length=1, radius=1, center=<1 0 0>, rotation=<rot -90 0 0>}]
 blade3_c_02 [{shapeType=Capsule, length=1, radius=1, center=<1 0 0>, rotation=<rot -90 0 0>}]
 blade3_d_01 [{shapeType=Capsule, length=1, radius=1, center=<1 0 0>, rotation=<rot -90 0 0>}]
 blade3_d_02 [{shapeType=Capsule, length=1, radius=1, center=<1 0 0>, rotation=<rot -90 0 0>}]
 blade3_e_01 [{shapeType=Capsule, length=1, radius=1, center=<1 0 0>, rotation=<rot -90 0 0>}]
 blade3_e_02 [{shapeType=Capsule, length=1, radius=1, center=<1 0 0>, rotation=<rot -90 0 0>}]
}
constraints [
 {bone1Name=Melee_TwoHanded_Handle, bone2Name=b_AR_Root, swing1LimitDegrees=10, swing2LimitDegrees=10, twistLimitDegrees=5}
 {bone1Name=blade1_base, bone2Name=Melee_TwoHanded_Handle, swing1LimitDegrees=1, swing2LimitDegrees=1, twistLimitDegrees=1}
 {bone1Name=blade2_base, bone2Name=Melee_TwoHanded_Handle, swing1LimitDegrees=1, swing2LimitDegrees=1, twistLimitDegrees=1}
 {bone1Name=blade3_base, bone2Name=Melee_TwoHanded_Handle, swing1LimitDegrees=1, swing2LimitDegrees=1, twistLimitDegrees=1}
 {bone1Name=blade1_a_01, bone2Name=blade1_base, swing1LimitDegrees=5, swing2LimitDegrees=5, twistLimitDegrees=5}
 {bone1Name=blade1_b_01, bone2Name=blade1_base, swing1LimitDegrees=10, swing2LimitDegrees=10, twistLimitDegrees=5}
 {bone1Name=blade1_c_01, bone2Name=blade1_base, swing1LimitDegrees=10, swing2LimitDegrees=10, twistLimitDegrees=5}
 {bone1Name=blade1_d_01, bone2Name=blade1_base, swing1LimitDegrees=10, swing2LimitDegrees=10, twistLimitDegrees=5}
 {bone1Name=blade1_e_01, bone2Name=blade1_base, swing1LimitDegrees=10, swing2LimitDegrees=10, twistLimitDegrees=5}
 {bone1Name=blade1_f_01, bone2Name=blade1_base, swing1LimitDegrees=10, swing2LimitDegrees=10, twistLimitDegrees=5}
 {bone1Name=blade2_a_02, bone2Name=blade2_base, swing1LimitDegrees=5, swing2LimitDegrees=5, twistLimitDegrees=5}
 {bone1Name=blade2_b_01, bone2Name=blade2_base, swing1LimitDegrees=10, swing2LimitDegrees=10, twistLimitDegrees=5}
 {bone1Name=blade2_c_01, bone2Name=blade2_base, swing1LimitDegrees=10, swing2LimitDegrees=10, twistLimitDegrees=5}
 {bone1Name=blade2_d_01, bone2Name=blade2_base, swing1LimitDegrees=10, swing2LimitDegrees=10, twistLimitDegrees=5}
 {bone1Name=blade2_e_01, bone2Name=blade2_base, swing1LimitDegrees=10, swing2LimitDegrees=10, twistLimitDegrees=5}
 {bone1Name=blade2_f_01, bone2Name=blade2_base, swing1LimitDegrees=10, swing2LimitDegrees=10, twistLimitDegrees=5}
 {bone1Name=blade3_a_01, bone2Name=blade3_base, swing1LimitDegrees=5, swing2LimitDegrees=5, twistLimitDegrees=5}
 {bone1Name=blade3_b_01, bone2Name=blade3_base, swing1LimitDegrees=5, swing2LimitDegrees=5, twistLimitDegrees=5}
 {bone1Name=blade3_c_01, bone2Name=blade3_base, swing1LimitDegrees=10, swing2LimitDegrees=10, twistLimitDegrees=5}
 {bone1Name=blade3_d_01, bone2Name=blade3_base, swing1LimitDegrees=10, swing2LimitDegrees=10, twistLimitDegrees=5}
 {bone1Name=blade3_e_01, bone2Name=blade3_base, swing1LimitDegrees=10, swing2LimitDegrees=10, twistLimitDegrees=5}
 {bone1Name=blade1_a_02, bone2Name=blade1_a_01, swing1LimitDegrees=5, swing2LimitDegrees=5, twistLimitDegrees=5}
 {bone1Name=blade1_b_02, bone2Name=blade1_b_01, swing1LimitDegrees=10, swing2LimitDegrees=10, twistLimitDegrees=5}
 {bone1Name=blade1_c_02, bone2Name=blade1_c_01, swing1LimitDegrees=10, swing2LimitDegrees=10, twistLimitDegrees=5}
 {bone1Name=blade1_d_02, bone2Name=blade1_d_01, swing1LimitDegrees=10, swing2LimitDegrees=10, twistLimitDegrees=5}
 {bone1Name=blade1_e_02, bone2Name=blade1_e_01, swing1LimitDegrees=10, swing2LimitDegrees=10, twistLimitDegrees=5}
 {bone1Name=blade1_f_02, bone2Name=blade1_f_01, swing1LimitDegrees=10, swing2LimitDegrees=10, twistLimitDegrees=5}
 {bone1Name=blade2_a_03, bone2Name=blade2_a_02, swing1LimitDegrees=5, swing2LimitDegrees=5, twistLimitDegrees=5}
 {bone1Name=blade2_b_02, bone2Name=blade2_b_01, swing1LimitDegrees=10, swing2LimitDegrees=10, twistLimitDegrees=5}
 {bone1Name=blade2_c_02, bone2Name=blade2_c_01, swing1LimitDegrees=10, swing2LimitDegrees=10, twistLimitDegrees=5}
 {bone1Name=blade2_d_02, bone2Name=blade2_d_01, swing1LimitDegrees=10, swing2LimitDegrees=10, twistLimitDegrees=5}
 {bone1Name=blade2_e_02, bone2Name=blade2_e_01, swing1LimitDegrees=10, swing2LimitDegrees=10, twistLimitDegrees=5}
 {bone1Name=blade2_f_02, bone2Name=blade2_f_01, swing1LimitDegrees=10, swing2LimitDegrees=10, twistLimitDegrees=5}
 {bone1Name=blade3_b_02, bone2Name=blade3_b_01, swing1LimitDegrees=10, swing2LimitDegrees=10, twistLimitDegrees=5}
 {bone1Name=blade3_a_02, bone2Name=blade3_a_01, swing1LimitDegrees=10, swing2LimitDegrees=10, twistLimitDegrees=5}
 {bone1Name=blade3_c_02, bone2Name=blade3_c_01, swing1LimitDegrees=10, swing2LimitDegrees=10, twistLimitDegrees=5}
 {bone1Name=blade3_d_02, bone2Name=blade3_d_01, swing1LimitDegrees=10, swing2LimitDegrees=10, twistLimitDegrees=5}
 {bone1Name=blade3_e_02, bone2Name=blade3_e_01, swing1LimitDegrees=10, swing2LimitDegrees=10, twistLimitDegrees=5}
]
