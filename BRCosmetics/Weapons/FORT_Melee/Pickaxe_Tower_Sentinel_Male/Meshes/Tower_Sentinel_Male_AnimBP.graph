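# Animation Blueprint for the Tower Sentinel Male pickaxe, layering physics and procedural floating motion onto its blade bones.
# Each update, AmbientMotion turns game-time sine waves into two sets of translation and rotation offsets, one time-shifted.
# The AnimGraph copies the mesh pose, runs RigidBody physics, then chains additive bone transforms for ambient motion and fixed pose corrections.
# Physics alpha follows the inverted front-end flag, clamped to stay partly active there.
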AnimBlueprint /BRCosmetics/Weapons/FORT_Melee/Pickaxe_Tower_Sentinel_Male/Meshes/Tower_Sentinel_Male_AnimBP
target_skeleton=&/Tower_Sentinel_Male_Skeleton
graphs [
 AmbientMotion {nodes [
   n1 = AmbientMotion!() -> (
    then
    Period:Vector="1.5,1,2"
    PeriodMult:double=1
    TimeOffset:double=~
    Magnitude:Vector="0.3,0.56,0.22"
    MagnitudeMult:double=1
    RotateMult:double=1
   ) {is=title K2Node_FunctionEntry}
   n2 = BreakVector(InVec=@n1.Magnitude)
   n3 = BreakVector(InVec=@n1)
   n4 = GetGameTimeInSeconds()
   n5 = float * float(A=float * float(A=@n1.PeriodMult, B=@n3.Y), B=@n4)
   n6 = float * float(A="Sin (Radians)"(A=float + float(A=@n5, B=@n1.TimeOffset) {is=title K2Node_CommutativeAssociativeBinaryOperator}), B=@n2.Y)
   n7 = float * float(A=float * float(A=@n1.PeriodMult, B=@n3.Z), B=@n4)
   n8 = float * float(A="Sin (Radians)"(A=float + float(A=@n7, B=@n1.TimeOffset)), B=@n2.Z)
   n9 = float * float(A=float * float(A=@n1.PeriodMult, B=@n3), B=@n4)
   n10 = float * float(A="Sin (Radians)"(A=float + float(A=@n9, B=@n1.TimeOffset)), B=@n2)
   n11 = MakeRotator(Roll=@n10, Pitch=@n6, Yaw=@n8)
   > Return Node!(
    execute
    VectorOut:Vector=vector * float(A=MakeVector(X=@n10, Y=@n6, Z=@n8), B=@n1.MagnitudeMult)
    RotateOut:Rotator="Scale Rotator (Float)"!(A:Rotator="Scale Rotator (Float)"(A=@n11, B=@n1.MagnitudeMult), B:float=@n1.RotateMult) -> (ReturnValue:Rotator)
   )
  ]}
 AnimGraph {nodes [
   n1 = RigidBody!(
    ComponentPose=Local To Component(LocalPose=Copy Pose From Mesh())
    Alpha=Get Is In Front End() -> (Is In Front End:double)
    OverrideWorldGravity:Vector="0,0,-100"
    ExternalForce
   ) -> (Pose) {
    bOverrideWorldGravity=true
    overridePhysicsAsset=&/Tower_Sentinel_Male_Physics
    alphaBoolBlend {blendOption=Linear}
    alphaScaleBiasClamp {bClampResult=true, clampMin=0.35}
    baseBoneRef {boneName=b_AR_Root}
    componentAppliedLinearAccClamp=<1500 1500 1500>
    componentLinearAccScale=<0.3 0.3 0.35>
    componentLinearVelScale=<0.3 0.3 0.35>
    overrideWorldGravity {z=-980}
   }
   n2 = "Transform (Modify) Bone - Bone: blade1_d_01"(
    ComponentPose=@n1
    Alpha=1
    Translation=Get ambientTrans_A_05() -> (ambientTrans_A_05:Vector)
    Rotation=Get ambientRot_A_05() -> (ambientRot_A_05:Rotator)
    Scale="1,1,1"
   ) {
    rotationMode=BMM_Additive
    rotationSpace=BCS_BoneSpace
    translationMode=BMM_Additive
    translationSpace=BCS_BoneSpace
    alphaBoolBlend {blendOption=Linear}
    boneToModify {boneName=blade1_d_01}
   }
   n3 = "Transform (Modify) Bone - Bone: blade1_d_02"(
    ComponentPose=@n2
    Alpha=1
    Translation=Get ambientTrans_A_03() -> (ambientTrans_A_03:Vector)
    Rotation=Get ambientRot_A_03() -> (ambientRot_A_03:Rotator)
    Scale="1,1,1"
   ) {
    rotationMode=BMM_Additive
    rotationSpace=BCS_BoneSpace
    translationMode=BMM_Additive
    translationSpace=BCS_BoneSpace
    alphaBoolBlend {blendOption=Linear}
    boneToModify {boneName=blade1_d_02}
   }
   n4 = "Transform (Modify) Bone - Bone: blade1_d_03"(
    ComponentPose=@n3
    Alpha=1
    Translation=Get ambientTrans_A_02() -> (ambientTrans_A_02:Vector)
    Rotation=Get ambientRot_A_02() -> (ambientRot_A_02:Rotator)
    Scale="1,1,1"
   ) {
    rotationMode=BMM_Additive
    rotationSpace=BCS_BoneSpace
    translationMode=BMM_Additive
    translationSpace=BCS_BoneSpace
    alphaBoolBlend {blendOption=Linear}
    boneToModify {boneName=blade1_d_03}
   }
   n5 = "Transform (Modify) Bone - Bone: blade1_d_04"(
    ComponentPose=@n4
    Alpha=1
    Translation=Get ambientTrans_A_07() -> (ambientTrans_A_07:Vector)
    Rotation=Get ambientRot_A_07() -> (ambientRot_A_07:Rotator)
    Scale="1,1,1"
   ) {
    rotationMode=BMM_Additive
    rotationSpace=BCS_BoneSpace
    translationMode=BMM_Additive
    translationSpace=BCS_BoneSpace
    alphaBoolBlend {blendOption=Linear}
    boneToModify {boneName=blade1_d_04}
   }
   n6 = "Transform (Modify) Bone - Bone: blade1_c_01"(
    ComponentPose=@n5
    Alpha=1
    Translation=Get ambientTrans_A_01() -> (ambientTrans_A_01:Vector)
    Rotation=Get ambientRot_A_01() -> (ambientRot_A_01:Rotator)
    Scale="1,1,1"
   ) {
    rotationMode=BMM_Additive
    rotationSpace=BCS_BoneSpace
    translationMode=BMM_Additive
    translationSpace=BCS_BoneSpace
    alphaBoolBlend {blendOption=Linear}
    boneToModify {boneName=blade1_c_01}
   }
   n7 = "Transform (Modify) Bone - Bone: blade1_c_02"(
    ComponentPose=@n6
    Alpha=1
    Translation=Get ambientTrans_A_06() -> (ambientTrans_A_06:Vector)
    Rotation=Get ambientRot_A_06() -> (ambientRot_A_06:Rotator)
    Scale="1,1,1"
   ) {
    rotationMode=BMM_Additive
    rotationSpace=BCS_BoneSpace
    translationMode=BMM_Additive
    translationSpace=BCS_BoneSpace
    alphaBoolBlend {blendOption=Linear}
    boneToModify {boneName=blade1_c_02}
   }
   n8 = "Transform (Modify) Bone - Bone: blade1_b_01"(
    ComponentPose=@n7
    Alpha=1
    Translation=Get ambientTrans_A_04() -> (ambientTrans_A_04:Vector)
    Rotation=Get ambientRot_A_04() -> (ambientRot_A_04:Rotator)
    Scale="1,1,1"
   ) {
    rotationMode=BMM_Additive
    rotationSpace=BCS_BoneSpace
    translationMode=BMM_Additive
    translationSpace=BCS_BoneSpace
    alphaBoolBlend {blendOption=Linear}
    boneToModify {boneName=blade1_b_01}
   }
   n9 = "Transform (Modify) Bone - Bone: blade1_b_02"(ComponentPose=@n8, Alpha=1, Translation=Get ambientTrans_A_05(), Rotation=Get ambientRot_A_05(), Scale="1,1,1") {
    rotationMode=BMM_Additive
    rotationSpace=BCS_BoneSpace
    translationMode=BMM_Additive
    translationSpace=BCS_BoneSpace
    alphaBoolBlend {blendOption=Linear}
    boneToModify {boneName=blade1_b_02}
   }
   n10 = "Transform (Modify) Bone - Bone: blade1_e_01"(ComponentPose=@n9, Alpha=1, Translation=Get ambientTrans_A_04(), Rotation=Get ambientRot_A_04(), Scale="1,1,1") {
    rotationMode=BMM_Additive
    rotationSpace=BCS_BoneSpace
    translationMode=BMM_Additive
    translationSpace=BCS_BoneSpace
    alphaBoolBlend {blendOption=Linear}
    boneToModify {boneName=blade1_e_01}
   }
   n11 = "Transform (Modify) Bone - Bone: blade1_f_01"(ComponentPose=@n10, Alpha=1, Translation=Get ambientTrans_A_02(), Rotation=Get ambientRot_A_02(), Scale="1,1,1") {
    rotationMode=BMM_Additive
    rotationSpace=BCS_BoneSpace
    translationMode=BMM_Additive
    translationSpace=BCS_BoneSpace
    alphaBoolBlend {blendOption=Linear}
    boneToModify {boneName=blade1_f_01}
   }
   n12 = "Transform (Modify) Bone - Bone: blade1_a_02"(ComponentPose=@n11, Alpha=1, Translation=Get ambientTrans_A_03(), Rotation=Get ambientRot_A_03(), Scale="1,1,1") {
    rotationMode=BMM_Additive
    rotationSpace=BCS_BoneSpace
    translationMode=BMM_Additive
    translationSpace=BCS_BoneSpace
    alphaBoolBlend {blendOption=Linear}
    boneToModify {boneName=blade1_a_02}
   }
   n13 = "Transform (Modify) Bone - Bone: blade1_c_02"(ComponentPose=@n12, Alpha=1, Translation=Get ambientTrans_A_01(), Rotation=Get ambientRot_A_01(), Scale="1,1,1") {
    rotationMode=BMM_Additive
    rotationSpace=BCS_BoneSpace
    translationMode=BMM_Additive
    translationSpace=BCS_BoneSpace
    alphaBoolBlend {blendOption=Linear}
    boneToModify {boneName=blade1_c_02}
   }
   n14 = "Transform (Modify) Bone - Bone: blade2_a_02"(
    ComponentPose=@n13
    Alpha=1
    Translation=Get ambientTrans_B_01() -> (ambientTrans_B_01:Vector)
    Rotation=Get ambientRot_B_01() -> (ambientRot_B_01:Rotator)
    Scale="1,1,1"
   ) {
    rotationMode=BMM_Additive
    rotationSpace=BCS_BoneSpace
    translationMode=BMM_Additive
    translationSpace=BCS_BoneSpace
    alphaBoolBlend {blendOption=Linear}
    boneToModify {boneName=blade2_a_02}
   }
   n15 = "Transform (Modify) Bone - Bone: blade2_a_03"(
    ComponentPose=@n14
    Alpha=1
    Translation=Get ambientTrans_B_04() -> (ambientTrans_B_04:Vector)
    Rotation=Get ambientRot_B_04() -> (ambientRot_B_04:Rotator)
    Scale="1,1,1"
   ) {
    rotationMode=BMM_Additive
    rotationSpace=BCS_BoneSpace
    translationMode=BMM_Additive
    translationSpace=BCS_BoneSpace
    alphaBoolBlend {blendOption=Linear}
    boneToModify {boneName=blade2_a_03}
   }
   n16 = "Transform (Modify) Bone - Bone: blade2_b_01"(
    ComponentPose=@n15
    Alpha=1
    Translation=Get ambientTrans_B_07() -> (ambientTrans_B_07:Vector)
    Rotation=Get ambientRot_B_07() -> (ambientRot_B_07:Rotator)
    Scale="1,1,1"
   ) {
    rotationMode=BMM_Additive
    rotationSpace=BCS_BoneSpace
    translationMode=BMM_Additive
    translationSpace=BCS_BoneSpace
    alphaBoolBlend {blendOption=Linear}
    boneToModify {boneName=blade2_b_01}
   }
   n17 = "Transform (Modify) Bone - Bone: blade2_c_01"(ComponentPose=@n16, Alpha=1, Translation=Get ambientTrans_B_07(), Rotation=Get ambientRot_B_07(), Scale="1,1,1") {
    rotationMode=BMM_Additive
    rotationSpace=BCS_BoneSpace
    translationMode=BMM_Additive
    translationSpace=BCS_BoneSpace
    alphaBoolBlend {blendOption=Linear}
    boneToModify {boneName=blade2_c_01}
   }
   n18 = "Transform (Modify) Bone - Bone: blade2_d_01"(
    ComponentPose=@n17
    Alpha=1
    Translation=Get ambientTrans_B_06() -> (ambientTrans_B_06:Vector)
    Rotation=Get ambientRot_B_06() -> (ambientRot_B_06:Rotator)
    Scale="1,1,1"
   ) {
    rotationMode=BMM_Additive
    rotationSpace=BCS_BoneSpace
    translationMode=BMM_Additive
    translationSpace=BCS_BoneSpace
    alphaBoolBlend {blendOption=Linear}
    boneToModify {boneName=blade2_d_01}
   }
   n19 = "Transform (Modify) Bone - Bone: blade2_e_01"(
    ComponentPose=@n18
    Alpha=1
    Translation=Get ambientTrans_B_05() -> (ambientTrans_B_05:Vector)
    Rotation=Get ambientRot_B_05() -> (ambientRot_B_05:Rotator)
    Scale="1,1,1"
   ) {
    rotationMode=BMM_Additive
    rotationSpace=BCS_BoneSpace
    translationMode=BMM_Additive
    translationSpace=BCS_BoneSpace
    alphaBoolBlend {blendOption=Linear}
    boneToModify {boneName=blade2_e_01}
   }
   n20 = "Transform (Modify) Bone - Bone: blade2_e_02"(ComponentPose=@n19, Alpha=1, Translation=Get ambientTrans_B_04(), Rotation=Get ambientRot_B_04(), Scale="1,1,1") {
    rotationMode=BMM_Additive
    rotationSpace=BCS_BoneSpace
    translationMode=BMM_Additive
    translationSpace=BCS_BoneSpace
    alphaBoolBlend {blendOption=Linear}
    boneToModify {boneName=blade2_e_02}
   }
   n21 = "Transform (Modify) Bone - Bone: blade2_f_01"(
    ComponentPose=@n20
    Alpha=1
    Translation=Get ambientTrans_B_03() -> (ambientTrans_B_03:Vector)
    Rotation=Get ambientRot_B_03() -> (ambientRot_B_03:Rotator)
    Scale="1,1,1"
   ) {
    rotationMode=BMM_Additive
    rotationSpace=BCS_BoneSpace
    translationMode=BMM_Additive
    translationSpace=BCS_BoneSpace
    alphaBoolBlend {blendOption=Linear}
    boneToModify {boneName=blade2_f_01}
   }
   n22 = "Transform (Modify) Bone - Bone: blade2_f_02"(
    ComponentPose=@n21
    Alpha=1
    Translation=Get ambientTrans_B_02() -> (ambientTrans_B_02:Vector)
    Rotation=Get ambientRot_B_02() -> (ambientRot_B_02:Rotator)
    Scale="1,1,1"
   ) {
    rotationMode=BMM_Additive
    rotationSpace=BCS_BoneSpace
    translationMode=BMM_Additive
    translationSpace=BCS_BoneSpace
    alphaBoolBlend {blendOption=Linear}
    boneToModify {boneName=blade2_f_02}
   }
   n23 = "Transform (Modify) Bone - Bone: blade2_f_03"(ComponentPose=@n22, Alpha=1, Translation=Get ambientTrans_B_01(), Rotation=Get ambientRot_B_01(), Scale="1,1,1") {
    rotationMode=BMM_Additive
    rotationSpace=BCS_BoneSpace
    translationMode=BMM_Additive
    translationSpace=BCS_BoneSpace
    alphaBoolBlend {blendOption=Linear}
    boneToModify {boneName=blade2_f_03}
   }
   n24 = "Transform (Modify) Bone - Bone: blade3_a_01"(ComponentPose=@n23, Alpha=1, Translation=Get ambientTrans_A_01(), Rotation=Get ambientRot_A_01(), Scale="1,1,1") {
    rotationMode=BMM_Additive
    rotationSpace=BCS_BoneSpace
    translationMode=BMM_Additive
    translationSpace=BCS_BoneSpace
    alphaBoolBlend {blendOption=Linear}
    boneToModify {boneName=blade3_a_01}
   }
   n25 = "Transform (Modify) Bone - Bone: blade3_b_01"(ComponentPose=@n24, Alpha=1, Translation=Get ambientTrans_A_02(), Rotation=Get ambientRot_A_02(), Scale="1,1,1") {
    rotationMode=BMM_Additive
    rotationSpace=BCS_BoneSpace
    translationMode=BMM_Additive
    translationSpace=BCS_BoneSpace
    alphaBoolBlend {blendOption=Linear}
    boneToModify {boneName=blade3_b_01}
   }
   n26 = "Transform (Modify) Bone - Bone: blade3_c_01"(ComponentPose=@n25, Alpha=1, Translation=Get ambientTrans_A_03(), Rotation=Get ambientRot_A_03(), Scale="1,1,1") {
    rotationMode=BMM_Additive
    rotationSpace=BCS_BoneSpace
    translationMode=BMM_Additive
    translationSpace=BCS_BoneSpace
    alphaBoolBlend {blendOption=Linear}
    boneToModify {boneName=blade3_c_01}
   }
   n27 = "Transform (Modify) Bone - Bone: blade3_d_01"(ComponentPose=@n26, Alpha=1, Translation=Get ambientTrans_A_04(), Rotation=Get ambientRot_A_04(), Scale="1,1,1") {
    rotationMode=BMM_Additive
    rotationSpace=BCS_BoneSpace
    translationMode=BMM_Additive
    translationSpace=BCS_BoneSpace
    alphaBoolBlend {blendOption=Linear}
    boneToModify {boneName=blade3_d_01}
   }
   n28 = "Transform (Modify) Bone - Bone: blade3_e_01"(ComponentPose=@n27, Alpha=1, Translation=Get ambientTrans_A_03(), Rotation=Get ambientRot_A_02(), Scale="1,1,1") {
    rotationMode=BMM_Additive
    rotationSpace=BCS_BoneSpace
    translationMode=BMM_Additive
    translationSpace=BCS_BoneSpace
    alphaBoolBlend {blendOption=Linear}
    boneToModify {boneName=blade3_e_01}
   }
   n29 = "Transform (Modify) Bone - Bone: blade3_e_02"(ComponentPose=@n28, Alpha=1, Translation=Get ambientTrans_A_05(), Rotation=Get ambientRot_A_05(), Scale="1,1,1") {
    rotationMode=BMM_Additive
    rotationSpace=BCS_BoneSpace
    translationMode=BMM_Additive
    translationSpace=BCS_BoneSpace
    alphaBoolBlend {blendOption=Linear}
    boneToModify {boneName=blade3_e_02}
   }
   n30 = "Transform (Modify) Bone - Bone: blade3_e_03"(ComponentPose=@n29, Alpha=1, Translation=Get ambientTrans_A_06(), Rotation=Get ambientRot_A_06(), Scale="1,1,1") {
    rotationMode=BMM_Additive
    rotationSpace=BCS_BoneSpace
    translationMode=BMM_Additive
    translationSpace=BCS_BoneSpace
    alphaBoolBlend {blendOption=Linear}
    boneToModify {boneName=blade3_e_03}
   }
   n31 = "Transform (Modify) Bone - Bone: blade3_base"(ComponentPose=@n30, Alpha=1, Translation="0,0,0", Rotation="0,0,10", Scale="1,1,1") {
    rotationMode=BMM_Additive
    rotationSpace=BCS_BoneSpace
    translationMode=BMM_Additive
    translationSpace=BCS_BoneSpace
    alphaBoolBlend {blendOption=Linear}
    boneToModify {boneName=blade3_base}
    rotation {roll=10}
   }
   n32 = "Transform (Modify) Bone - Bone: blade3_e_01"(ComponentPose=@n31, Alpha=1, Translation="0,0,0", Rotation="0,0,0", Scale="1,0,1") {
    rotationMode=BMM_Additive
    rotationSpace=BCS_BoneSpace
    translationMode=BMM_Additive
    translationSpace=BCS_BoneSpace
    alphaBoolBlend {blendOption=Linear}
    boneToModify {boneName=blade3_e_01}
    scale {y=0}
   }
   n33 = "Transform (Modify) Bone - Bone: blade3_e_02"(ComponentPose=@n32, Alpha=1, Translation="1,0,0", Rotation="0,-10,0", Scale="1,0,1") {
    rotationMode=BMM_Additive
    rotationSpace=BCS_BoneSpace
    translationMode=BMM_Additive
    translationSpace=BCS_BoneSpace
    alphaBoolBlend {blendOption=Linear}
    boneToModify {boneName=blade3_e_02}
    rotation {yaw=-10}
    scale {y=0}
    translation {x=1}
   }
   n34 = "Transform (Modify) Bone - Bone: blade3_e_03"(ComponentPose=@n33, Alpha=1, Translation="1,0,0", Rotation="0,15,0", Scale="1,0,1") {
    rotationMode=BMM_Additive
    rotationSpace=BCS_BoneSpace
    translationMode=BMM_Additive
    translationSpace=BCS_BoneSpace
    alphaBoolBlend {blendOption=Linear}
    boneToModify {boneName=blade3_e_03}
    rotation {yaw=15}
    scale {y=0}
    translation {x=1}
   }
   n35 = "Transform (Modify) Bone - Bone: blade3_c_01"(ComponentPose=@n34, Alpha=1, Translation="3.5,1.5,0", Rotation="0,25,0", Scale="1,0,1") {
    rotationMode=BMM_Additive
    rotationSpace=BCS_BoneSpace
    translationMode=BMM_Additive
    translationSpace=BCS_BoneSpace
    alphaBoolBlend {blendOption=Linear}
    boneToModify {boneName=blade3_c_01}
    rotation {yaw=25}
    scale {y=0}
    translation=<3.5 1.5>
   }
   n36 = "Transform (Modify) Bone - Bone: blade3_d_01"(ComponentPose=@n35, Alpha=1, Translation="2,-4,0", Rotation="0,-15,0", Scale="1,0,1") {
    rotationMode=BMM_Additive
    rotationSpace=BCS_BoneSpace
    translationMode=BMM_Additive
    translationSpace=BCS_BoneSpace
    alphaBoolBlend {blendOption=Linear}
    boneToModify {boneName=blade3_d_01}
    rotation {yaw=-15}
    scale {y=0}
    translation=<2 -4>
   }
   n37 = "Transform (Modify) Bone - Bone: blade3_b_01"(ComponentPose=@n36, Alpha=1, Translation="0,0,0", Rotation="0,-5,0", Scale="1,1,1") {
    rotationMode=BMM_Additive
    rotationSpace=BCS_BoneSpace
    translationMode=BMM_Additive
    translationSpace=BCS_BoneSpace
    alphaBoolBlend {blendOption=Linear}
    boneToModify {boneName=blade3_b_01}
    rotation {yaw=-5}
   }
   n38 = "Transform (Modify) Bone - Bone: blade3_a_01"(ComponentPose=@n37, Alpha=1, Translation="0,0,0", Rotation="-0.025,-10,0.005", Scale="1,1,1") {
    rotationMode=BMM_Additive
    rotationSpace=BCS_BoneSpace
    translationMode=BMM_Additive
    translationSpace=BCS_BoneSpace
    alphaBoolBlend {blendOption=Linear}
    boneToModify {boneName=blade3_a_01}
    rotation=<rot -0.025 -10 0.005>
   }
   n39 = "Transform (Modify) Bone - Bone: blade1_a_01"(ComponentPose=@n38, Alpha=1, Translation="0,0,0", Rotation="0,-50,0", Scale="1,1,1") {
    rotationMode=BMM_Additive
    rotationSpace=BCS_BoneSpace
    translationMode=BMM_Additive
    translationSpace=BCS_BoneSpace
    alphaBoolBlend {blendOption=Linear}
    boneToModify {boneName=blade1_a_01}
    rotation {yaw=-50}
   }
   n40 = "Transform (Modify) Bone - Bone: blade1_a_02"(ComponentPose=@n39, Alpha=1, Translation="1.5,-1.5,0", Rotation="-15,-50,10", Scale="1,1,1") {
    rotationMode=BMM_Additive
    rotationSpace=BCS_BoneSpace
    translationMode=BMM_Additive
    translationSpace=BCS_BoneSpace
    alphaBoolBlend {blendOption=Linear}
    boneToModify {boneName=blade1_a_02}
    rotation=<rot -15 -50 10>
    translation=<1.5 -1.5>
   }
   n41 = "Transform (Modify) Bone - Bone: blade1_b_01"(ComponentPose=@n40, Alpha=1, Translation="0,0,0", Rotation="0,-25,0", Scale="1,1,1") {
    rotationMode=BMM_Additive
    rotationSpace=BCS_BoneSpace
    translationMode=BMM_Additive
    translationSpace=BCS_BoneSpace
    alphaBoolBlend {blendOption=Linear}
    boneToModify {boneName=blade1_b_01}
    rotation {yaw=-25}
   }
   n42 = "Transform (Modify) Bone - Bone: blade1_b_02"(ComponentPose=@n41, Alpha=1, Translation="0,0,0", Rotation="0,0,0", Scale="1,1,1") {
    rotationMode=BMM_Additive
    rotationSpace=BCS_BoneSpace
    translationMode=BMM_Additive
    translationSpace=BCS_BoneSpace
    alphaBoolBlend {blendOption=Linear}
    boneToModify {boneName=blade1_b_02}
   }
   n43 = "Transform (Modify) Bone - Bone: blade1_e_01"(ComponentPose=@n42, Alpha=1, Translation="1,0.5,0", Rotation="0,-10,0", Scale="1,1,1") {
    rotationMode=BMM_Additive
    rotationSpace=BCS_BoneSpace
    translationMode=BMM_Additive
    translationSpace=BCS_BoneSpace
    alphaBoolBlend {blendOption=Linear}
    boneToModify {boneName=blade1_e_01}
    rotation {yaw=-10}
    translation=<1 0.5>
   }
   n44 = "Transform (Modify) Bone - Bone: blade1_c_01"(ComponentPose=@n43, Alpha=1, Translation="0,0,0", Rotation="0,-20,0", Scale="1,1,1") {
    rotationMode=BMM_Additive
    rotationSpace=BCS_BoneSpace
    translationMode=BMM_Additive
    translationSpace=BCS_BoneSpace
    alphaBoolBlend {blendOption=Linear}
    boneToModify {boneName=blade1_c_01}
    rotation {yaw=-20}
   }
   n45 = "Transform (Modify) Bone - Bone: blade1_c_02"(ComponentPose=@n44, Alpha=1, Translation="0,0,0", Rotation="0,10,0", Scale="1,1,1") {
    rotationMode=BMM_Additive
    rotationSpace=BCS_BoneSpace
    translationMode=BMM_Additive
    translationSpace=BCS_BoneSpace
    alphaBoolBlend {blendOption=Linear}
    boneToModify {boneName=blade1_c_02}
    rotation {yaw=10}
   }
   n46 = "Transform (Modify) Bone - Bone: blade1_f_01"(ComponentPose=@n45, Alpha=1, Translation="0,0,0", Rotation="0,5,0", Scale="1,1,1") {
    rotationMode=BMM_Additive
    rotationSpace=BCS_BoneSpace
    translationMode=BMM_Additive
    translationSpace=BCS_BoneSpace
    alphaBoolBlend {blendOption=Linear}
    boneToModify {boneName=blade1_f_01}
    rotation {yaw=5}
   }
   n47 = "Transform (Modify) Bone - Bone: blade1_d_01"(ComponentPose=@n46, Alpha=1, Translation="0,0,0", Rotation="0,30,0", Scale="1,1,1") {
    rotationMode=BMM_Additive
    rotationSpace=BCS_BoneSpace
    translationMode=BMM_Additive
    translationSpace=BCS_BoneSpace
    alphaBoolBlend {blendOption=Linear}
    boneToModify {boneName=blade1_d_01}
    rotation {yaw=30}
   }
   n48 = "Transform (Modify) Bone - Bone: blade1_d_02"(ComponentPose=@n47, Alpha=1, Translation="1,-0.5,0", Rotation="0,-45,0", Scale="1,1,1") {
    rotationMode=BMM_Additive
    rotationSpace=BCS_BoneSpace
    translationMode=BMM_Additive
    translationSpace=BCS_BoneSpace
    alphaBoolBlend {blendOption=Linear}
    boneToModify {boneName=blade1_d_02}
    rotation {yaw=-45}
    translation=<1 -0.5>
   }
   n49 = "Transform (Modify) Bone - Bone: blade2_a_02"(ComponentPose=@n48, Alpha=1, Translation="0,0,0", Rotation="0,-50,0", Scale="1,1,1") {
    rotationMode=BMM_Additive
    rotationSpace=BCS_BoneSpace
    translationMode=BMM_Additive
    translationSpace=BCS_BoneSpace
    alphaBoolBlend {blendOption=Linear}
    boneToModify {boneName=blade2_a_02}
    rotation {yaw=-50}
   }
   n50 = "Transform (Modify) Bone - Bone: blade2_a_03"(ComponentPose=@n49, Alpha=1, Translation="0,-1.5,1.5", Rotation="15,-50,10", Scale="1,1,1") {
    rotationMode=BMM_Additive
    rotationSpace=BCS_BoneSpace
    translationMode=BMM_Additive
    translationSpace=BCS_BoneSpace
    alphaBoolBlend {blendOption=Linear}
    boneToModify {boneName=blade2_a_03}
    rotation=<rot 15 -50 10>
    translation {y=-1.5, z=1.5}
   }
   n51 = "Transform (Modify) Bone - Bone: blade2_c_01"(ComponentPose=@n50, Alpha=1, Translation="0,0,0", Rotation="0,-25,0", Scale="1,1,1") {
    rotationMode=BMM_Additive
    rotationSpace=BCS_BoneSpace
    translationMode=BMM_Additive
    translationSpace=BCS_BoneSpace
    alphaBoolBlend {blendOption=Linear}
    boneToModify {boneName=blade2_c_01}
    rotation {yaw=-25}
   }
   n52 = "Transform (Modify) Bone - Bone: blade2_c_02"(ComponentPose=@n51, Alpha=1, Translation="0,0,0", Rotation="0,0,0", Scale="1,1,1") {
    rotationMode=BMM_Additive
    rotationSpace=BCS_BoneSpace
    translationMode=BMM_Additive
    translationSpace=BCS_BoneSpace
    alphaBoolBlend {blendOption=Linear}
    boneToModify {boneName=blade2_c_02}
   }
   n53 = "Transform (Modify) Bone - Bone: blade2_b_01"(ComponentPose=@n52, Alpha=1, Translation="1,0.5,0", Rotation="0,-10,0", Scale="1,1,1") {
    rotationMode=BMM_Additive
    rotationSpace=BCS_BoneSpace
    translationMode=BMM_Additive
    translationSpace=BCS_BoneSpace
    alphaBoolBlend {blendOption=Linear}
    boneToModify {boneName=blade2_b_01}
    rotation {yaw=-10}
    translation=<1 0.5>
   }
   n54 = "Transform (Modify) Bone - Bone: blade2_e_01"(ComponentPose=@n53, Alpha=1, Translation="0,0,0", Rotation="0,-20,0", Scale="1,1,1") {
    rotationMode=BMM_Additive
    rotationSpace=BCS_BoneSpace
    translationMode=BMM_Additive
    translationSpace=BCS_BoneSpace
    alphaBoolBlend {blendOption=Linear}
    boneToModify {boneName=blade2_e_01}
    rotation {yaw=-20}
   }
   n55 = "Transform (Modify) Bone - Bone: blade2_e_02"(ComponentPose=@n54, Alpha=1, Translation="0,0,0", Rotation="0,10,0", Scale="1,1,1") {
    rotationMode=BMM_Additive
    rotationSpace=BCS_BoneSpace
    translationMode=BMM_Additive
    translationSpace=BCS_BoneSpace
    alphaBoolBlend {blendOption=Linear}
    boneToModify {boneName=blade2_e_02}
    rotation {yaw=10}
   }
   n56 = "Transform (Modify) Bone - Bone: blade2_d_01"(ComponentPose=@n55, Alpha=1, Translation="0,0,0", Rotation="0,5,0", Scale="1,1,1") {
    rotationMode=BMM_Additive
    rotationSpace=BCS_BoneSpace
    translationMode=BMM_Additive
    translationSpace=BCS_BoneSpace
    alphaBoolBlend {blendOption=Linear}
    boneToModify {boneName=blade2_d_01}
    rotation {yaw=5}
   }
   n57 = "Transform (Modify) Bone - Bone: blade2_f_01"(ComponentPose=@n56, Alpha=1, Translation="0,0,0", Rotation="0,30,0", Scale="1,1,1") {
    rotationMode=BMM_Additive
    rotationSpace=BCS_BoneSpace
    translationMode=BMM_Additive
    translationSpace=BCS_BoneSpace
    alphaBoolBlend {blendOption=Linear}
    boneToModify {boneName=blade2_f_01}
    rotation {yaw=30}
   }
   n58 = "Transform (Modify) Bone - Bone: blade2_f_02"(ComponentPose=@n57, Alpha=1, Translation="1,-0.5,0", Rotation="0,-45,0", Scale="1,1,1") {
    rotationMode=BMM_Additive
    rotationSpace=BCS_BoneSpace
    translationMode=BMM_Additive
    translationSpace=BCS_BoneSpace
    alphaBoolBlend {blendOption=Linear}
    boneToModify {boneName=blade2_f_02}
    rotation {yaw=-45}
    translation=<1 -0.5>
   }
   Output Pose(Result=Component To Local(ComponentPose=@n58)) {layerGroup=None}
  ]}
 EventGraph {nodes [
   Event BlueprintUpdateAnimation()
   > n1 = Sequence!(execute) -> (then_0, then_1, then_2:Exec, then_3:Exec, then_4:Exec, then_5:Exec, then_6:Exec, then_7:Exec, then_8:Exec)
   > n2 = AmbientMotion!(
    execute
    self
    Period:Vector="1.5,1,2"
    PeriodMult:double=0.9
    TimeOffset:double=0
    Magnitude:Vector="0.7,0.56,0.22"
    MagnitudeMult:double=0.7
    RotateMult:double=10
   ) -> (then, VectorOut:Vector, RotateOut:Rotator) {is=title K2Node_CallFunction}
   > Set ambientTrans_A_01(ambientTrans_A_01:Vector=@n2) -> (Output_Get:Vector)
   > Set ambientRot_A_01(ambientRot_A_01:Rotator=@n2.RotateOut) -> (Output_Get:Rotator)
   n3 = AmbientMotion(execute=@n1.then_1, Period="1.5,1,2", PeriodMult=0.7, TimeOffset=0, Magnitude="0.7,0.56,0.22", MagnitudeMult=0.85, RotateMult=12)
   > Set ambientTrans_A_02(ambientTrans_A_02:Vector=@n3) -> (Output_Get:Vector)
   > Set ambientRot_A_02(ambientRot_A_02:Rotator=@n3.RotateOut) -> (Output_Get:Rotator)
   n4 = AmbientMotion(execute=@n1.then_2, Period="1.5,1,2", PeriodMult=0.9, TimeOffset=0, Magnitude="0.7,0.56,0.22", MagnitudeMult=0.8, RotateMult=14)
   > Set ambientTrans_A_03(ambientTrans_A_03:Vector=@n4) -> (Output_Get:Vector)
   > Set ambientRot_A_03(ambientRot_A_03:Rotator=@n4.RotateOut) -> (Output_Get:Rotator)
   n5 = AmbientMotion(execute=@n1.then_3, Period="1.5,1,2", PeriodMult=0.9, TimeOffset=0, Magnitude="0.7,0.56,0.22", MagnitudeMult=0.5, RotateMult=16)
   > Set ambientTrans_A_04(ambientTrans_A_04:Vector=@n5) -> (Output_Get:Vector)
   > Set ambientRot_A_04(ambientRot_A_04:Rotator=@n5.RotateOut) -> (Output_Get:Rotator)
   n6 = AmbientMotion(execute=@n1.then_4, Period="1.5,1,2", PeriodMult=0.9, TimeOffset=0, Magnitude="0.7,0.56,0.22", MagnitudeMult=0.7, RotateMult=10)
   > Set ambientTrans_A_05(ambientTrans_A_05:Vector=@n6) -> (Output_Get:Vector)
   > Set ambientRot_A_05(ambientRot_A_05:Rotator=@n6.RotateOut) -> (Output_Get:Rotator)
   n7 = AmbientMotion(execute=@n1.then_5, Period="1.5,1,2", PeriodMult=0.7, TimeOffset=0, Magnitude="0.7,0.56,0.22", MagnitudeMult=0.85, RotateMult=12)
   > Set ambientTrans_A_06(ambientTrans_A_06:Vector=@n7) -> (Output_Get:Vector)
   > Set ambientRot_A_06(ambientRot_A_06:Rotator=@n7.RotateOut) -> (Output_Get:Rotator)
   n8 = AmbientMotion(execute=@n1.then_6, Period="1.5,1,2", PeriodMult=0.9, TimeOffset=0, Magnitude="0.7,0.56,0.22", MagnitudeMult=0.8, RotateMult=14)
   > Set ambientTrans_A_07(ambientTrans_A_07:Vector=@n8) -> (Output_Get:Vector)
   > Set ambientRot_A_07(ambientRot_A_07:Rotator=@n8.RotateOut) -> (Output_Get:Rotator)
   n9 = Sequence!(execute=@n1.then_7) -> (then_0, then_1, then_2, then_3, then_4, then_5, then_6, then_7)
   > n10 = AmbientMotion(
    Period="2,1.5,1"
    PeriodMult=0.9
    TimeOffset=Get timeOffset_0() -> (timeOffset_0:double)
    Magnitude="0.7,0.56,0.22"
    MagnitudeMult=0.7
    RotateMult=10
   )
   > Set ambientTrans_B_01(ambientTrans_B_01:Vector=@n10) -> (Output_Get:Vector)
   > Set ambientRot_B_01(ambientRot_B_01:Rotator=@n10.RotateOut) -> (Output_Get:Rotator)
   n11 = AmbientMotion(
    execute=@n9.then_1
    Period="2,1.5,1"
    PeriodMult=0.7
    TimeOffset=Get timeOffset_1() -> (timeOffset_1:double)
    Magnitude="0.7,0.56,0.22"
    MagnitudeMult=0.85
    RotateMult=12
   )
   > Set ambientTrans_B_02(ambientTrans_B_02:Vector=@n11) -> (Output_Get:Vector)
   > Set ambientRot_B_02(ambientRot_B_02:Rotator=@n11.RotateOut) -> (Output_Get:Rotator)
   n12 = AmbientMotion(
    execute=@n9.then_2
    Period="2,1.5,1"
    PeriodMult=0.9
    TimeOffset=Get timeOffset_2() -> (timeOffset_2:double)
    Magnitude="0.7,0.56,0.22"
    MagnitudeMult=0.8
    RotateMult=14
   )
   > Set ambientTrans_B_03(ambientTrans_B_03:Vector=@n12) -> (Output_Get:Vector)
   > Set ambientRot_B_03(ambientRot_B_03:Rotator=@n12.RotateOut) -> (Output_Get:Rotator)
   n13 = AmbientMotion(
    execute=@n9.then_3
    Period="2,1.5,1"
    PeriodMult=0.9
    TimeOffset=Get timeOffset_3() -> (timeOffset_3:double)
    Magnitude="0.7,0.56,0.22"
    MagnitudeMult=0.5
    RotateMult=16
   )
   > Set ambientTrans_B_04(ambientTrans_B_04:Vector=@n13) -> (Output_Get:Vector)
   > Set ambientRot_B_04(ambientRot_B_04:Rotator=@n13.RotateOut) -> (Output_Get:Rotator)
   n14 = AmbientMotion(
    execute=@n9.then_4
    Period="2,1.5,1"
    PeriodMult=0.9
    TimeOffset=Get timeOffset_4() -> (timeOffset_4:double)
    Magnitude="0.7,0.56,0.22"
    MagnitudeMult=0.7
    RotateMult=16
   )
   > Set ambientTrans_B_05(ambientTrans_B_05:Vector=@n14) -> (Output_Get:Vector)
   > Set ambientRot_B_05(ambientRot_B_05:Rotator=@n14.RotateOut) -> (Output_Get:Rotator)
   n15 = AmbientMotion(
    execute=@n9.then_5
    Period="2,1.5,1"
    PeriodMult=0.7
    TimeOffset=Get timeOffset_5() -> (timeOffset_5:double)
    Magnitude="0.7,0.56,0.22"
    MagnitudeMult=0.85
    RotateMult=12
   )
   > Set ambientTrans_B_06(ambientTrans_B_06:Vector=@n15) -> (Output_Get:Vector)
   > Set ambientRot_B_06(ambientRot_B_06:Rotator=@n15.RotateOut) -> (Output_Get:Rotator)
   n16 = AmbientMotion(
    execute=@n9.then_6
    Period="2,1.5,1"
    PeriodMult=0.9
    TimeOffset=Get timeOffset_6() -> (timeOffset_6:double)
    Magnitude="0.7,0.56,0.22"
    MagnitudeMult=0.8
    RotateMult=14
   )
   > Set ambientTrans_B_07(ambientTrans_B_07:Vector=@n16) -> (Output_Get:Vector)
   > Set ambientRot_B_07(ambientRot_B_07:Rotator=@n16.RotateOut) -> (Output_Get:Rotator)
   n17 = Get bIsInFrontEnd()
   Set Is In Front End(execute=@n1.then_8, Is In Front End:double="To Float (Boolean)"(InBool=NOT Boolean(A=@n17))) -> (Output_Get:double)
   TryGetPawnOwner()
  ]}
]
variables [
 ambientRot_A_01 {type=Rotator}
 ambientRot_A_02 {type=Rotator}
 ambientRot_A_03 {type=Rotator}
 ambientRot_A_04 {type=Rotator}
 ambientRot_A_05 {type=Rotator}
 ambientRot_A_06 {type=Rotator}
 ambientRot_A_07 {type=Rotator}
 ambientRot_B_01 {type=Rotator}
 ambientRot_B_02 {type=Rotator}
 ambientRot_B_03 {type=Rotator}
 ambientRot_B_04 {type=Rotator}
 ambientRot_B_05 {type=Rotator}
 ambientRot_B_06 {type=Rotator}
 ambientRot_B_07 {type=Rotator}
 ambientTrans_A_01 {type=Vector}
 ambientTrans_A_02 {type=Vector}
 ambientTrans_A_03 {type=Vector}
 ambientTrans_A_04 {type=Vector}
 ambientTrans_A_05 {type=Vector}
 ambientTrans_A_06 {type=Vector}
 ambientTrans_A_07 {type=Vector}
 ambientTrans_B_01 {type=Vector}
 ambientTrans_B_02 {type=Vector}
 ambientTrans_B_03 {type=Vector}
 ambientTrans_B_04 {type=Vector}
 ambientTrans_B_05 {type=Vector}
 ambientTrans_B_06 {type=Vector}
 ambientTrans_B_07 {type=Vector}
 Is In Front End {type=number}
 NewVar {type=boolean}
 timeOffset_0 {type=number, default=5}
 timeOffset_1 {type=number, default=8}
 timeOffset_2 {type=number, default=11}
 timeOffset_3 {type=number, default=15}
 timeOffset_4 {type=number, default=16}
 timeOffset_5 {type=number, default=18}
 timeOffset_6 {type=number, default=21}
]
